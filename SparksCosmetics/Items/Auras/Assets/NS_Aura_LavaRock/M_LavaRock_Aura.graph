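# Lit lava rock material for Niagara mesh particles in the LavaRock aura cosmetic, reusing the MagicStone aura textures.
# Emissive combines a particle-alpha-driven Fresnel tint with particle-colored glow, built from Fresnel color plus a panned lava mask, multiplied by the emissive texture.
# The lava mask pans over bounding-box-based coordinates.
# Dynamic parameters control strength, frequency and randomness of a panning Perlin-noise jitter in X and Y, intended for world position offset.
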Material /SparksCosmetics/Items/Auras/Assets/NS_Aura_LavaRock/M_LavaRock_Aura
graph {
 properties {bAutomaticallySetUsageInEditor=false, bUsedWithNiagaraMeshParticles=true, usedShadingModels=Default Lit}
 nodes [
  n1 = Constant()
  n2 = VectorParameter() {parameterName=Fresnel Color, defaultValue=<rgba 0.9115 0.0414 0 1>}
  n3 = ParticleColor()
  n4 = Multiply(A=@n3.A, B=Fresnel() {exponent=1})
  n5 = VectorParameter() {parameterName=Glow Noise Color, defaultValue=<rgba 1 0.4893 0 1>}
  n6 = Fresnel() {baseReflectFraction=0.005}
  n7 = Panner(Coordinate=MaterialFunctionCall() {materialFunction=&BoundingBoxBased_0-1_UVW}.B) {speedX=0.1, speedY=0.3}
  n8 = Multiply(
   A=Add(
    A=Multiply(A=@n5.RGB, B=@n6)
    B=TextureSample(UVs=@n7) {samplerType=SAMPLERTYPE_Masks, texture=/Game/Environments/Sets/Vulcano/Textures/Lava_M}.R
   )
   B=TextureSampleParameter2D() {parameterName=Emissive, texture=/SparksCosmetics/Items/Auras/Assets/NS_Aura_Magicstone/T_MagicStone_Aura_E}.RGB
  )
  MaterialFunctionCall(Emissive=Add(A=LinearInterpolate(A=@n1, B=@n2.RGB, Alpha=@n4), B=Multiply(A=@n3.RGB, B=@n8))) {materialFunction=&MF_FX_EmissiveAdjustments}
  n9 = DynamicParameter() {defaultValue=<rgba 0 1 1 1>, paramNames [WPO Jitter Strength, WPO Jitter Frequency, WPO Jitter Random, Param4]}
  n10 = Panner(Coordinate=@n9.WPO Jitter Frequency) {speedX=2, speedY=0.5}
  n11 = Multiply(A=TextureSample(UVs=Add(A=@n10, B=@n9.WPO Jitter Random)) {samplerType=SAMPLERTYPE_LinearGrayscale, texture=&T_SimplePerlin_Greyscale}.G) {constB=2}
  n12 = MaterialFunctionCall(Float3=Multiply(A=@n9.WPO Jitter Strength, B=Add(A=@n11) {constB=-1})) {materialFunction=&BreakOutFloat3Components}
  MaterialFunctionCall(X=@n12.R, Y=@n12.R, Z=Constant()) {materialFunction=&MakeFloat3}
  TextureSampleParameter2D() {parameterName=Normal, samplerType=SAMPLERTYPE_Normal, texture=/SparksCosmetics/Items/Auras/Assets/NS_Aura_Magicstone/T_MagicStone_Aura_N}
  TextureSampleParameter2D() {parameterName=Diffuse, texture=/SparksCosmetics/Items/Auras/Assets/NS_Aura_Magicstone/T_MagicStone_Aura_D}
  TextureSample() {samplerType=SAMPLERTYPE_Masks, texture=/SparksCosmetics/Items/Auras/Assets/NS_Aura_Magicstone/T_MagicStone_Aura_S}
 ]
}
